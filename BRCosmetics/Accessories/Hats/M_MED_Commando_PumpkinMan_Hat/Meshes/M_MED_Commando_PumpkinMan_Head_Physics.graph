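# Physics asset for the Commando PumpkinMan head and hat cosmetic, defining the bodies and joints for the hat's dynamic bones.
# A capsule sits on the head bone and on each of three hat bones, dyn_hat_02 through dyn_hat_04, tapering along the chain.
# Constraints link each hat body to the one before it, starting at the head, with equal moderate swing and twist limit angles.
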PhysicsAsset /BRCosmetics/Accessories/Hats/M_MED_Commando_PumpkinMan_Hat/Meshes/M_MED_Commando_PumpkinMan_Head_Physics
bodies {
 dyn_hat_02 [{shapeType=Capsule, length=4.9998, radius=2.8667, center=<-0.0073 -0.5609 -0.4151>, rotation=<rot -1.0109 0 -25.2975>}]
 dyn_hat_03 [{shapeType=Capsule, length=2.3776, radius=1.9475, center=<0.6975 -0.4441 -0.3372>, rotation=<rot 0.0001 57.4646 77.926>}]
 dyn_hat_04 [{shapeType=Capsule, length=2.25, radius=1.75, center=<0.6646 -0.5234 -0.0009>, rotation=<rot 0.0001 51.7923 89.9998>}]
 head [{shapeType=Capsule, length=2.7733, radius=3.1233}]
}
constraints [
 {bone1Name=dyn_hat_02, bone2Name=head, swing1LimitDegrees=45, swing2LimitDegrees=45, twistLimitDegrees=45}
 {bone1Name=dyn_hat_03, bone2Name=dyn_hat_02, swing1LimitDegrees=45, swing2LimitDegrees=45, twistLimitDegrees=45}
 {bone1Name=dyn_hat_04, bone2Name=dyn_hat_03, swing1LimitDegrees=45, swing2LimitDegrees=45, twistLimitDegrees=45}
]
