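# Post-process style AnimBP for the Commando PumpkinMan hat head mesh, adding secondary motion to hat, mask and hair bones.
# Copies the parent mesh pose, then chains AnimDynamics, a low-alpha RigidBody, head-rotation-driven Flap nodes, collision-limited hair dynamics and a bone offset.
# Update event scales hat-tip alpha by pawn speed, setting hair alpha only when stationary.
# Some Flap chains, one hair node and copy_alpha logic are unwired.
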AnimBlueprint /BRCosmetics/Accessories/Hats/M_MED_Commando_PumpkinMan_Hat/Meshes/M_MED_Commando_PumpkinMan_Head_AnimBP
target_skeleton=&/M_MED_Commando_PumpkinMan_Head_Skeleton
graphs [
 AnimGraph {nodes [
   n1 = Get hair_alpha() -> (hair_alpha:double)
   Anim Dynamics - Bone: dyn_hair_r_03(Alpha=@n1, ExternalForce="0,0,0") {
    angularDampingOverride=0.99
    angularSpringConstant=1000
    bAngularSpring=true
    bGravityOverrideInSimSpace=true
    bOverrideAngularDamping=true
    bUsePlanarLimit=false
    bUseSphericalLimits=true
    gravityScale=0.35
    alphaBoolBlend {blendOption=Linear}
    boundBone {boneName=dyn_hair_r_03}
    componentAppliedLinearAccClamp=<70 70 70>
    componentLinearVelScale=<300 300 300>
    gravityOverride {z=-980}
    physicsBodyDefinitions [{
      collisionType=CustomSphere
      sphereCollisionRadius=1.5
      boundBone {boneName=dyn_hair_r_03}
      boxExtents=<10 30 30>
      constraintSetup {angularLimitsMax=<0 45 5>, angularLimitsMin=<0 -2 -30>, angularTarget=<0 0 0>}
      localJointOffset=<-1.5 0 0>
     }]
    sphericalLimits [
     {limitRadius=60, drivingBone {boneName=spine_05}, sphereLocalOffset=<-48 -6.5 0>}
     {limitRadius=9, drivingBone {boneName=spine_05}, sphereLocalOffset=<13 0 0>}
    ]
   }
   n2 = Anim Dynamics - Bone: dyn_hat_strandA(
    ComponentPose=Local To Component(LocalPose=Copy Pose From Mesh() {bCopyCurves=true, bUseAttachedParent=true})
    Alpha=0.3
    ExternalForce="0,0,0"
   ) {
    angularDampingOverride=0.985
    angularSpringConstant=150
    bAngularSpring=true
    bGravityOverrideInSimSpace=true
    bLinearSpring=true
    bOverrideAngularDamping=true
    bOverrideLinearDamping=true
    gravityScale=0.35
    linearDampingOverride=0.99
    linearSpringConstant=120
    simulationSpace=RootRelative
    alphaBoolBlend {blendOption=Linear}
    boundBone {boneName=dyn_hat_strandA}
    componentAppliedLinearAccClamp=<500 500 500>
    componentLinearAccScale=<1 1 1>
    componentLinearVelScale=<50 50 50>
    physicsBodyDefinitions [{
      sphereCollisionRadius=0
      boundBone {boneName=dyn_hat_strandA}
      boxExtents=<30 30 30>
      constraintSetup {
       angularTargetAxis=AxisY
       angularLimitsMax=<0 0 0>
       angularLimitsMin=<0 0 0>
       angularTarget=<0 0 0>
       linearAxesMax=<5.3768 6.3298 5>
       linearAxesMin=<-4.6209 -6.2114 -6>
      }
      localJointOffset=<0 0 0>
     }]
   }
   n3 = Anim Dynamics - Bone: dyn_hat_strandB(ComponentPose=@n2, Alpha=0.45, ExternalForce="0,0,0") {
    angularDampingOverride=0.985
    angularSpringConstant=160
    bAngularSpring=true
    bGravityOverrideInSimSpace=true
    bLinearSpring=true
    bOverrideAngularDamping=true
    bOverrideLinearDamping=true
    gravityScale=0.3
    linearDampingOverride=0.99
    linearSpringConstant=185
    simulationSpace=RootRelative
    alphaBoolBlend {blendOption=Linear}
    boundBone {boneName=dyn_hat_strandB}
    componentAppliedLinearAccClamp=<500 500 500>
    componentLinearAccScale=<1 1 1>
    componentLinearVelScale=<20 20 20>
    physicsBodyDefinitions [{
      sphereCollisionRadius=0
      boundBone {boneName=dyn_hat_strandB}
      boxExtents=<20 20 20>
      constraintSetup {
       angularTargetAxis=AxisY
       angularLimitsMax=<0 0 0>
       angularLimitsMin=<0 0 0>
       angularTarget=<0 0 0>
       linearAxesMax=<2.6406 5.3414 4>
       linearAxesMin=<-3.0262 -4.6437 -3>
      }
      localJointOffset=<0 0 0>
     }]
   }
   n4 = Anim Dynamics - Bone: dyn_hat_leafA(ComponentPose=@n3, Alpha=0.35, ExternalForce="0,0,0") {
    angularDampingOverride=0.985
    angularSpringConstant=255
    bAngularSpring=true
    bGravityOverrideInSimSpace=true
    bLinearSpring=true
    bOverrideAngularDamping=true
    bOverrideLinearDamping=true
    gravityScale=0.15
    linearDampingOverride=0.99
    linearSpringConstant=200
    alphaBoolBlend {blendOption=Linear}
    boundBone {boneName=dyn_hat_leafA}
    componentAppliedLinearAccClamp=<1000 1000 1000>
    componentLinearVelScale=<100 100 100>
    physicsBodyDefinitions [{
      sphereCollisionRadius=0
      boundBone {boneName=dyn_hat_leafA}
      boxExtents=<30 30 30>
      constraintSetup {
       angularTargetAxis=AxisY
       angularLimitsMax=<0 0 0>
       angularLimitsMin=<0 0 0>
       angularTarget=<0 0 0>
       linearAxesMax=<2 2 0.5>
       linearAxesMin=<-1.5 0 -0.5>
      }
      localJointOffset=<0 0 0>
     }]
   }
   n5 = Anim Dynamics - Bone: dyn_hat_leafB(ComponentPose=@n4, Alpha=0.3, ExternalForce="0,0,0") {
    angularDampingOverride=0.985
    angularSpringConstant=190
    bAngularSpring=true
    bGravityOverrideInSimSpace=true
    bLinearSpring=true
    bOverrideAngularDamping=true
    bOverrideLinearDamping=true
    gravityScale=0.2
    linearDampingOverride=0.99
    linearSpringConstant=175
    alphaBoolBlend {blendOption=Linear}
    boundBone {boneName=dyn_hat_leafB}
    componentAppliedLinearAccClamp=<1000 1000 1000>
    componentLinearVelScale=<100 100 100>
    physicsBodyDefinitions [{
      sphereCollisionRadius=0
      boundBone {boneName=dyn_hat_leafB}
      boxExtents=<30 30 30>
      constraintSetup {
       angularTargetAxis=AxisY
       angularLimitsMax=<0 0 0>
       angularLimitsMin=<0 0 0>
       angularTarget=<0 0 0>
       linearAxesMax=<2 2 1.5>
       linearAxesMin=<-2 -1 -1.5>
      }
      localJointOffset=<0 0 0>
     }]
   }
   n6 = Anim Dynamics - Bone: dyn_hat_leafC(ComponentPose=@n5, Alpha=0.35, ExternalForce="0,0,0") {
    angularDampingOverride=0.985
    angularSpringConstant=200
    bAngularSpring=true
    bGravityOverrideInSimSpace=true
    bLinearSpring=true
    bOverrideAngularDamping=true
    bOverrideLinearDamping=true
    gravityScale=0.2
    linearDampingOverride=0.99
    linearSpringConstant=150
    alphaBoolBlend {blendOption=Linear}
    boundBone {boneName=dyn_hat_leafC}
    componentAppliedLinearAccClamp=<1000 1000 1000>
    componentLinearVelScale=<100 100 100>
    physicsBodyDefinitions [{
      sphereCollisionRadius=0
      boundBone {boneName=dyn_hat_leafC}
      boxExtents=<20 10 15>
      constraintSetup {
       angularTargetAxis=AxisY
       angularLimitsMax=<0 0 0>
       angularLimitsMin=<0 0 0>
       angularTarget=<0 0 0>
       linearAxesMax=<3 2 0.5>
       linearAxesMin=<-0.5 0 -0.5>
      }
      localJointOffset=<0 0 0>
     }]
   }
   n7 = Anim Dynamics - Bone: dyn_hat_leafD(ComponentPose=@n6, Alpha=0.3, ExternalForce="0,0,0") {
    angularDampingOverride=0.985
    angularSpringConstant=150
    bAngularSpring=true
    bGravityOverrideInSimSpace=true
    bLinearSpring=true
    bOverrideAngularDamping=true
    bOverrideLinearDamping=true
    gravityScale=0.15
    linearDampingOverride=0.99
    linearSpringConstant=170
    alphaBoolBlend {blendOption=Linear}
    boundBone {boneName=dyn_hat_leafD}
    componentAppliedLinearAccClamp=<1000 1000 1000>
    componentLinearVelScale=<100 100 100>
    physicsBodyDefinitions [{
      sphereCollisionRadius=0
      boundBone {boneName=dyn_hat_leafD}
      boxExtents=<30 30 30>
      constraintSetup {
       angularTargetAxis=AxisY
       angularLimitsMax=<0 0 0>
       angularLimitsMin=<0 0 0>
       angularTarget=<0 0 0>
       linearAxesMax=<2 2 1.5>
       linearAxesMin=<-2 -1 -1.5>
      }
      localJointOffset=<0 0 0>
     }]
   }
   n8 = Anim Dynamics - Bone: dyn_hat_leafE(ComponentPose=@n7, Alpha=0.3, ExternalForce="0,0,0") {
    angularDampingOverride=0.985
    angularSpringConstant=300
    bAngularSpring=true
    bGravityOverrideInSimSpace=true
    bLinearSpring=true
    bOverrideAngularDamping=true
    bOverrideLinearDamping=true
    gravityScale=0.15
    linearDampingOverride=0.99
    linearSpringConstant=200
    alphaBoolBlend {blendOption=Linear}
    boundBone {boneName=dyn_hat_leafE}
    componentAppliedLinearAccClamp=<1000 1000 1000>
    componentLinearVelScale=<100 100 100>
    physicsBodyDefinitions [{
      sphereCollisionRadius=0
      boundBone {boneName=dyn_hat_leafE}
      boxExtents=<30 30 30>
      constraintSetup {
       angularTargetAxis=AxisY
       angularLimitsMax=<0 0 0>
       angularLimitsMin=<0 0 0>
       angularTarget=<0 0 0>
       linearAxesMax=<2 2 1.5>
       linearAxesMin=<-2 -1 -1.5>
      }
      localJointOffset=<0 0 0>
     }]
   }
   n9 = Anim Dynamics - Bone: dyn_mask_a_l(ComponentPose=@n8, Alpha=0.35, ExternalForce="0,0,0") {
    angularDampingOverride=0.985
    angularSpringConstant=200
    bAngularSpring=true
    bGravityOverrideInSimSpace=true
    bLinearSpring=true
    bOverrideAngularDamping=true
    bOverrideLinearDamping=true
    gravityScale=0.25
    linearDampingOverride=0.99
    linearSpringConstant=150
    simulationSpace=RootRelative
    alphaBoolBlend {blendOption=Linear}
    boundBone {boneName=dyn_mask_a_l}
    componentAppliedLinearAccClamp=<500 500 500>
    componentLinearVelScale=<100 100 100>
    physicsBodyDefinitions [{
      sphereCollisionRadius=0
      boundBone {boneName=dyn_mask_a_l}
      boxExtents=<20 20 20>
      constraintSetup {
       angularTargetAxis=AxisY
       angularLimitsMax=<0 0 0>
       angularLimitsMin=<0 0 0>
       angularTarget=<0 0 0>
       linearAxesMax=<3 2 4>
       linearAxesMin=<-3 -2 -3>
      }
      localJointOffset=<0 0 0>
     }]
   }
   n10 = Anim Dynamics - Bone: dyn_mask_b_l(ComponentPose=@n9, Alpha=0.275, ExternalForce="0,0,0") {
    angularDampingOverride=0.985
    angularSpringConstant=200
    bAngularSpring=true
    bGravityOverrideInSimSpace=true
    bLinearSpring=true
    bOverrideAngularDamping=true
    bOverrideLinearDamping=true
    gravityScale=0.28
    linearDampingOverride=0.99
    linearSpringConstant=150
    simulationSpace=RootRelative
    alphaBoolBlend {blendOption=Linear}
    boundBone {boneName=dyn_mask_b_l}
    componentAppliedLinearAccClamp=<500 500 500>
    componentLinearVelScale=<100 100 100>
    physicsBodyDefinitions [{
      sphereCollisionRadius=0
      boundBone {boneName=dyn_mask_b_l}
      boxExtents=<25 25 25>
      constraintSetup {
       angularTargetAxis=AxisY
       angularLimitsMax=<0 0 0>
       angularLimitsMin=<0 0 0>
       angularTarget=<0 0 0>
       linearAxesMax=<3 2 4>
       linearAxesMin=<-3 -2 -3>
      }
      localJointOffset=<0 0 0>
     }]
   }
   n11 = Anim Dynamics - Bone: dyn_mask_a_r(ComponentPose=@n10, Alpha=0.35, ExternalForce="0,0,0") {
    angularDampingOverride=0.985
    angularSpringConstant=200
    bAngularSpring=true
    bGravityOverrideInSimSpace=true
    bLinearSpring=true
    bOverrideAngularDamping=true
    bOverrideLinearDamping=true
    gravityScale=0.2
    linearDampingOverride=0.99
    linearSpringConstant=150
    simulationSpace=RootRelative
    alphaBoolBlend {blendOption=Linear}
    boundBone {boneName=dyn_mask_a_r}
    componentAppliedLinearAccClamp=<500 500 500>
    componentLinearVelScale=<100 100 100>
    physicsBodyDefinitions [{
      sphereCollisionRadius=0
      boundBone {boneName=dyn_mask_a_r}
      boxExtents=<20 20 20>
      constraintSetup {
       angularTargetAxis=AxisY
       angularLimitsMax=<0 0 0>
       angularLimitsMin=<0 0 0>
       angularTarget=<0 0 0>
       linearAxesMax=<3 2 4>
       linearAxesMin=<-3 -2 -3>
      }
      localJointOffset=<0 0 0>
     }]
   }
   n12 = Anim Dynamics - Bone: dyn_mask_b_r(ComponentPose=@n11, Alpha=0.3, ExternalForce="0,0,0") {
    angularDampingOverride=0.985
    angularSpringConstant=200
    bAngularSpring=true
    bGravityOverrideInSimSpace=true
    bLinearSpring=true
    bOverrideAngularDamping=true
    bOverrideLinearDamping=true
    gravityScale=0.28
    linearDampingOverride=0.99
    linearSpringConstant=150
    simulationSpace=RootRelative
    alphaBoolBlend {blendOption=Linear}
    boundBone {boneName=dyn_mask_b_r}
    componentAppliedLinearAccClamp=<500 500 500>
    componentLinearVelScale=<100 100 100>
    physicsBodyDefinitions [{
      sphereCollisionRadius=0
      boundBone {boneName=dyn_mask_b_r}
      boxExtents=<25 25 25>
      constraintSetup {
       angularTargetAxis=AxisY
       angularLimitsMax=<0 0 0>
       angularLimitsMin=<0 0 0>
       angularTarget=<0 0 0>
       linearAxesMax=<3 2 4>
       linearAxesMin=<-3 -2 -3>
      }
      localJointOffset=<0 0 0>
     }]
   }
   n13 = Anim Dynamics - Bone: dyn_hat_pumpA(ComponentPose=@n12, Alpha=0.62, ExternalForce="0,0,0") {
    angularDampingOverride=0.985
    angularSpringConstant=120
    bAngularSpring=true
    bGravityOverrideInSimSpace=true
    bOverrideAngularDamping=true
    gravityScale=0.355
    alphaBoolBlend {blendOption=Linear}
    boundBone {boneName=dyn_hat_pumpA}
    componentAppliedLinearAccClamp=<1000 1000 1000>
    componentLinearVelScale=<5 5 5>
    physicsBodyDefinitions [{
      sphereCollisionRadius=0
      boundBone {boneName=dyn_hat_pumpA}
      boxExtents=<45 35 45>
      constraintSetup {angularLimitsMax=<20 22 19>, angularLimitsMin=<-20 -15 -15>, angularTarget=<0 0 0>}
      localJointOffset=<0 0 -10>
     }]
   }
   n14 = Anim Dynamics - Bone: dyn_hat_pumpB(ComponentPose=@n13, Alpha=0.575, ExternalForce="0,0,0") {
    angularDampingOverride=0.985
    angularSpringConstant=120
    bAngularSpring=true
    bGravityOverrideInSimSpace=true
    bOverrideAngularDamping=true
    gravityScale=0.275
    alphaBoolBlend {blendOption=Linear}
    boundBone {boneName=dyn_hat_pumpB}
    componentAppliedLinearAccClamp=<1000 1000 1000>
    componentLinearVelScale=<5 5 5>
    physicsBodyDefinitions [{
      sphereCollisionRadius=0
      boundBone {boneName=dyn_hat_pumpB}
      boxExtents=<35 35 35>
      constraintSetup {angularLimitsMax=<15 15 15>, angularLimitsMin=<-15 -15 -15>, angularTarget=<0 0 0>}
      localJointOffset=<0 0 -8>
     }]
   }
   n15 = Anim Dynamics - Bone: dyn_hat_tipA(ComponentPose=@n14, Alpha=Get speed_mult() -> (speed_mult:double), ExternalForce="0,0,0") {
    angularDampingOverride=0.985
    angularSpringConstant=100
    bAngularSpring=true
    bGravityOverrideInSimSpace=true
    bLinearSpring=true
    bOverrideAngularDamping=true
    bOverrideLinearDamping=true
    gravityScale=0.33
    linearDampingOverride=0.995
    linearSpringConstant=275
    simulationSpace=RootRelative
    alphaBoolBlend {blendOption=Linear}
    alphaScaleBiasClamp {bInterpResult=true, interpSpeedDecreasing=8, interpSpeedIncreasing=8}
    boundBone {boneName=dyn_hat_tipA}
    componentAppliedLinearAccClamp=<500 500 500>
    componentLinearAccScale=<1 1 1>
    componentLinearVelScale=<100 100 100>
    physicsBodyDefinitions [{
      sphereCollisionRadius=0
      boundBone {boneName=dyn_hat_tipA}
      boxExtents=<30 30 30>
      constraintSetup {
       angularTargetAxis=AxisY
       angularLimitsMax=<0 0 0>
       angularLimitsMin=<0 0 0>
       angularTarget=<0 0 0>
       linearAxesMax=<2.6406 3.5 4>
       linearAxesMin=<-3.0262 -3 -3>
      }
      localJointOffset=<0 0 0>
     }]
   }
   n16 = RigidBody(ComponentPose=@n15, Alpha=0.2) {
    bOverrideWorldGravity=true
    overridePhysicsAsset=&/M_MED_Commando_PumpkinMan_Head_Physics
    alphaBoolBlend {blendOption=Linear}
    componentAppliedLinearAccClamp=<100 100 100>
    componentLinearAccScale=<1 1 1>
    componentLinearVelScale=<150 150 150>
    overrideWorldGravity {z=-980}
   }
   n17 = Flap Node - Bone: hat_rear!(ComponentPose:Component Space Pose Link Structure=@n16, Alpha:float=1) -> (Pose:Component Space Pose Link Structure) {
    is=title FortAnimGraphNode_Flap
    bUseClamp=true
    targetClampMax=200
    targetComponent=RotationZ
    alphaBoolBlend {blendOption=Linear}
    sourceDrives [{
      bUseRange=false
      drivingCurve=null
      multiplier=2
      rangeMax=1
      rangeMin=0
      remappedMax=1
      remappedMin=0
      sourceComponent=RotationZ
      useQuaternion=true
      sourceBone {boneName=head}
     }]
    targetBone {boneName=hat_rear}
   }
   n18 = Flap Node - Bone: dyn_hair_bk_01!(ComponentPose:Component Space Pose Link Structure=@n17, Alpha:float=1) -> (Pose:Component Space Pose Link Structure) {
    is=title FortAnimGraphNode_Flap
    bUseClamp=true
    targetClampMax=100
    targetClampMin=-5
    targetComponent=TranslationY
    alphaBoolBlend {blendOption=Linear}
    sourceDrives [{
      bUseRange=false
      drivingCurve=null
      multiplier=6
      rangeMax=1
      rangeMin=0
      remappedMax=1
      remappedMin=0
      sourceComponent=RotationZ
      useQuaternion=true
      sourceBone {boneName=head}
     }]
    targetBone {boneName=dyn_hair_bk_01}
   }
   n19 = Flap Node - Bone: dyn_hair_bk_l_01!(ComponentPose:Component Space Pose Link Structure=@n18, Alpha:float=1) -> (Pose:Component Space Pose Link Structure) {
    is=title FortAnimGraphNode_Flap
    bUseClamp=true
    targetClampMax=100
    targetClampMin=-5
    targetComponent=TranslationY
    alphaBoolBlend {blendOption=Linear}
    sourceDrives [{
      bUseRange=false
      drivingCurve=null
      multiplier=6
      rangeMax=1
      rangeMin=0
      remappedMax=1
      remappedMin=0
      sourceComponent=RotationZ
      useQuaternion=true
      sourceBone {boneName=head}
     }]
    targetBone {boneName=dyn_hair_bk_l_01}
   }
   n20 = Flap Node - Bone: dyn_hair_bk_r_01!(ComponentPose:Component Space Pose Link Structure=@n19, Alpha:float=1) -> (Pose:Component Space Pose Link Structure) {
    is=title FortAnimGraphNode_Flap
    bUseClamp=true
    targetClampMax=100
    targetClampMin=-5
    targetComponent=TranslationY
    alphaBoolBlend {blendOption=Linear}
    sourceDrives [{
      bUseRange=false
      drivingCurve=null
      multiplier=6
      rangeMax=1
      rangeMin=0
      remappedMax=1
      remappedMin=0
      sourceComponent=RotationZ
      useQuaternion=true
      sourceBone {boneName=head}
     }]
    targetBone {boneName=dyn_hair_bk_r_01}
   }
   n21 = Flap Node - Bone: dyn_hair_bk_01(ComponentPose=@n20, Alpha=1) {
    bUseClamp=true
    targetClampMax=100
    targetClampMin=-5
    targetComponent=TranslationX
    alphaBoolBlend {blendOption=Linear}
    sourceDrives [{
      bUseRange=false
      drivingCurve=null
      multiplier=3
      rangeMax=1
      rangeMin=0
      remappedMax=1
      remappedMin=0
      sourceComponent=RotationZ
      useQuaternion=true
      sourceBone {boneName=head}
     }]
    targetBone {boneName=dyn_hair_bk_01}
   }
   n22 = Flap Node - Bone: dyn_hair_bk_l_01(ComponentPose=@n21, Alpha=1) {
    bUseClamp=true
    targetClampMax=100
    targetClampMin=-5
    targetComponent=TranslationX
    alphaBoolBlend {blendOption=Linear}
    sourceDrives [{
      bUseRange=false
      drivingCurve=null
      multiplier=3
      rangeMax=1
      rangeMin=0
      remappedMax=1
      remappedMin=0
      sourceComponent=RotationZ
      useQuaternion=true
      sourceBone {boneName=head}
     }]
    targetBone {boneName=dyn_hair_bk_l_01}
   }
   n23 = Flap Node - Bone: dyn_hair_bk_r_01(ComponentPose=@n22, Alpha=1) {
    bUseClamp=true
    targetClampMax=100
    targetClampMin=-5
    targetComponent=TranslationX
    alphaBoolBlend {blendOption=Linear}
    sourceDrives [{
      bUseRange=false
      drivingCurve=null
      multiplier=3
      rangeMax=1
      rangeMin=0
      remappedMax=1
      remappedMin=0
      sourceComponent=RotationZ
      useQuaternion=true
      sourceBone {boneName=head}
     }]
    targetBone {boneName=dyn_hair_bk_r_01}
   }
   n24 = Flap Node - Bone: dyn_hair_bk_02!(ComponentPose:Component Space Pose Link Structure=@n23, Alpha:float=1) -> (Pose:Component Space Pose Link Structure) {
    is=title FortAnimGraphNode_Flap
    bUseClamp=true
    targetClampMax=100
    targetClampMin=-5
    targetComponent=TranslationX
    alphaBoolBlend {blendOption=Linear}
    sourceDrives [{
      bUseRange=false
      drivingCurve=null
      multiplier=-3
      rangeMax=1
      rangeMin=0
      remappedMax=1
      remappedMin=0
      sourceComponent=RotationZ
      useQuaternion=true
      sourceBone {boneName=head}
     }]
    targetBone {boneName=dyn_hair_bk_02}
   }
   n25 = Flap Node - Bone: dyn_hair_bk_l_02!(ComponentPose:Component Space Pose Link Structure=@n24, Alpha:float=1) -> (Pose:Component Space Pose Link Structure) {
    is=title FortAnimGraphNode_Flap
    bUseClamp=true
    targetClampMax=100
    targetClampMin=-5
    targetComponent=TranslationX
    alphaBoolBlend {blendOption=Linear}
    sourceDrives [{
      bUseRange=false
      drivingCurve=null
      multiplier=-3
      rangeMax=1
      rangeMin=0
      remappedMax=1
      remappedMin=0
      sourceComponent=RotationZ
      useQuaternion=true
      sourceBone {boneName=head}
     }]
    targetBone {boneName=dyn_hair_bk_l_02}
   }
   n26 = Flap Node - Bone: dyn_hair_bk_r_02!(ComponentPose:Component Space Pose Link Structure=@n25, Alpha:float=1) -> (Pose:Component Space Pose Link Structure) {
    is=title FortAnimGraphNode_Flap
    bUseClamp=true
    targetClampMax=100
    targetClampMin=-5
    targetComponent=TranslationX
    alphaBoolBlend {blendOption=Linear}
    sourceDrives [{
      bUseRange=false
      drivingCurve=null
      multiplier=-3
      rangeMax=1
      rangeMin=0
      remappedMax=1
      remappedMin=0
      sourceComponent=RotationZ
      useQuaternion=true
      sourceBone {boneName=head}
     }]
    targetBone {boneName=dyn_hair_bk_r_02}
   }
   n27 = Anim Dynamics - Bone: dyn_hair_bk_01(ComponentPose=@n26, Alpha=@n1, ExternalForce="0,0,0") {
    angularDampingOverride=0.99
    angularSpringConstant=1000
    bAngularSpring=true
    bGravityOverrideInSimSpace=true
    bOverrideAngularDamping=true
    bUsePlanarLimit=false
    bUseSphericalLimits=true
    gravityScale=0.35
    alphaBoolBlend {blendOption=Linear}
    boundBone {boneName=dyn_hair_bk_01}
    componentAppliedLinearAccClamp=<70 70 70>
    componentLinearVelScale=<300 300 300>
    gravityOverride {z=-980}
    physicsBodyDefinitions [{
      collisionType=CustomSphere
      sphereCollisionRadius=4.3
      boundBone {boneName=dyn_hair_bk_01}
      boxExtents=<5 30 30>
      constraintSetup {angularLimitsMax=<0 20 25>, angularLimitsMin=<0 -20 -65>, angularTarget=<0 0 0>}
      localJointOffset=<6.5 0 0>
     }]
    planarLimits [{drivingBone {boneName=neck_01}, planeTransform {location=<15 -6 6>, rotation=<rot 90 -90 0>}}]
    sphericalLimits [{limitRadius=70, drivingBone {boneName=spine_05}, sphereLocalOffset=<-30 50 0>}]
   }
   n28 = Anim Dynamics - Bone: dyn_hair_bk_02(ComponentPose=@n27, Alpha=@n1, ExternalForce="0,0,0") {
    angularDampingOverride=0.99
    angularSpringConstant=1000
    bAngularSpring=true
    bGravityOverrideInSimSpace=true
    bOverrideAngularDamping=true
    bUseSphericalLimits=true
    gravityScale=0.35
    alphaBoolBlend {blendOption=Linear}
    boundBone {boneName=dyn_hair_bk_02}
    componentAppliedLinearAccClamp=<70 70 70>
    componentLinearVelScale=<300 300 300>
    gravityOverride {z=-980}
    physicsBodyDefinitions [{
      collisionType=CustomSphere
      sphereCollisionRadius=4
      boundBone {boneName=dyn_hair_bk_02}
      boxExtents=<5 40 40>
      constraintSetup {angularLimitsMax=<0 20 7>, angularLimitsMin=<0 -20 -75>, angularTarget=<0 0 0>}
      localJointOffset=<10 0 0>
     }]
    planarLimits [
     {drivingBone {boneName=neck_01}, planeTransform {location=<7 -8 6>, rotation=<rot -90 -90 0>}}
     {drivingBone {boneName=spine_05}, planeTransform {location=<7 -12 6>, rotation=<rot -90 -36.5651 0>}}
    ]
    sphericalLimits [
     {limitRadius=45, drivingBone {boneName=spine_05}, sphereLocalOffset=<-25 18 0>}
     {limitRadius=8, drivingBone {boneName=neck_02}, sphereLocalOffset=<0 -1.5 0>}
    ]
   }
   n29 = Anim Dynamics - Bone: dyn_hair_bk_r_01(ComponentPose=@n28, Alpha=@n1, ExternalForce="0,0,0") {
    angularDampingOverride=0.99
    angularSpringConstant=1000
    bAngularSpring=true
    bGravityOverrideInSimSpace=true
    bOverrideAngularDamping=true
    bUsePlanarLimit=false
    gravityScale=0.35
    alphaBoolBlend {blendOption=Linear}
    boundBone {boneName=dyn_hair_bk_r_01}
    componentAppliedLinearAccClamp=<70 70 70>
    componentLinearVelScale=<300 300 300>
    gravityOverride {z=-980}
    physicsBodyDefinitions [{
      collisionType=CustomSphere
      sphereCollisionRadius=3
      boundBone {boneName=dyn_hair_bk_r_01}
      boxExtents=<5 30 30>
      constraintSetup {angularLimitsMax=<0 10 20>, angularLimitsMin=<0 -10 -65>, angularTarget=<0 0 0>}
      localJointOffset=<6.5 0 0>
     }]
    sphericalLimits [{limitRadius=7, drivingBone {boneName=neck_02}, sphereLocalOffset=<0 0 0>}]
   }
   n30 = Anim Dynamics - Bone: dyn_hair_bk_r_02(ComponentPose=@n29, Alpha=@n1, ExternalForce="0,0,0") {
    angularDampingOverride=0.99
    angularSpringConstant=1000
    bAngularSpring=true
    bGravityOverrideInSimSpace=true
    bOverrideAngularDamping=true
    bUseSphericalLimits=true
    gravityScale=0.35
    alphaBoolBlend {blendOption=Linear}
    boundBone {boneName=dyn_hair_bk_r_02}
    componentAppliedLinearAccClamp=<70 70 70>
    componentLinearVelScale=<300 300 300>
    gravityOverride {z=-980}
    physicsBodyDefinitions [{
      collisionType=CustomSphere
      sphereCollisionRadius=4
      boundBone {boneName=dyn_hair_bk_r_02}
      boxExtents=<5 40 40>
      constraintSetup {angularLimitsMax=<0 10 7>, angularLimitsMin=<0 -10 -75>, angularTarget=<0 0 0>}
      localJointOffset=<10 0 0>
     }]
    sphericalLimits [{limitRadius=55, drivingBone {boneName=spine_05}, sphereLocalOffset=<-40 23 8>}]
   }
   n31 = Anim Dynamics - Bone: dyn_hair_bk_l_01(ComponentPose=@n30, Alpha=@n1, ExternalForce="0,0,0") {
    angularDampingOverride=0.99
    angularSpringConstant=1000
    bAngularSpring=true
    bGravityOverrideInSimSpace=true
    bOverrideAngularDamping=true
    bUsePlanarLimit=false
    gravityScale=0.35
    alphaBoolBlend {blendOption=Linear}
    boundBone {boneName=dyn_hair_bk_l_01}
    componentAppliedLinearAccClamp=<70 70 70>
    componentLinearVelScale=<300 300 300>
    gravityOverride {z=-980}
    physicsBodyDefinitions [{
      collisionType=CustomSphere
      sphereCollisionRadius=3
      boundBone {boneName=dyn_hair_bk_l_01}
      boxExtents=<5 30 30>
      constraintSetup {angularLimitsMax=<0 10 20>, angularLimitsMin=<0 -10 -65>, angularTarget=<0 0 0>}
      localJointOffset=<6.5 0 0>
     }]
    sphericalLimits [{limitRadius=8, drivingBone {boneName=neck_02}, sphereLocalOffset=<0 0 0>}]
   }
   n32 = Anim Dynamics - Bone: dyn_hair_bk_l_02(ComponentPose=@n31, Alpha=@n1, ExternalForce="0,0,0") {
    angularDampingOverride=0.99
    angularSpringConstant=1000
    bAngularSpring=true
    bGravityOverrideInSimSpace=true
    bOverrideAngularDamping=true
    bUsePlanarLimit=false
    bUseSphericalLimits=true
    gravityScale=0.35
    alphaBoolBlend {blendOption=Linear}
    boundBone {boneName=dyn_hair_bk_l_02}
    componentAppliedLinearAccClamp=<70 70 70>
    componentLinearVelScale=<300 300 300>
    gravityOverride {z=-980}
    physicsBodyDefinitions [{
      collisionType=CustomSphere
      sphereCollisionRadius=4
      boundBone {boneName=dyn_hair_bk_l_02}
      boxExtents=<5 40 40>
      constraintSetup {angularLimitsMax=<0 10 7>, angularLimitsMin=<0 -10 -75>, angularTarget=<0 0 0>}
      localJointOffset=<10 0 0>
     }]
    sphericalLimits [{limitRadius=46, drivingBone {boneName=spine_05}, sphereLocalOffset=<-30 19 -8>}]
   }
   n33 = Anim Dynamics - Bone: dyn_hair_l_01(ComponentPose=@n32, Alpha=@n1, ExternalForce="0,0,0") {
    angularDampingOverride=0.99
    angularSpringConstant=1000
    bAngularSpring=true
    bGravityOverrideInSimSpace=true
    bOverrideAngularDamping=true
    bUsePlanarLimit=false
    gravityScale=0.35
    alphaBoolBlend {blendOption=Linear}
    boundBone {boneName=dyn_hair_l_01}
    componentAppliedLinearAccClamp=<70 70 70>
    componentLinearVelScale=<300 300 300>
    gravityOverride {z=-980}
    physicsBodyDefinitions [{
      collisionType=CustomSphere
      sphereCollisionRadius=3
      boundBone {boneName=dyn_hair_l_01}
      boxExtents=<5 30 30>
      constraintSetup {angularLimitsMax=<0 8 3>, angularLimitsMin=<0 -8 -11>, angularTarget=<0 0 0>}
      localJointOffset=<-6.5 0 0>
     }]
    sphericalLimits [
     {limitRadius=8, drivingBone {boneName=neck_02}, sphereLocalOffset=<0 0 0>}
     {limitRadius=60, drivingBone {boneName=spine_05}, sphereLocalOffset=<-42 5 0>}
    ]
   }
   n34 = Anim Dynamics - Bone: dyn_hair_l_02(ComponentPose=@n33, Alpha=@n1, ExternalForce="0,0,0") {
    angularDampingOverride=0.99
    angularSpringConstant=200
    bAngularSpring=true
    bGravityOverrideInSimSpace=true
    bOverrideAngularDamping=true
    bUseSphericalLimits=true
    gravityScale=0.35
    alphaBoolBlend {blendOption=Linear}
    boundBone {boneName=dyn_hair_l_02}
    componentAppliedLinearAccClamp=<70 70 70>
    componentLinearVelScale=<300 300 300>
    gravityOverride {z=-980}
    physicsBodyDefinitions [{
      collisionType=CustomSphere
      sphereCollisionRadius=2.25
      boundBone {boneName=dyn_hair_l_02}
      boxExtents=<10 30 30>
      constraintSetup {angularLimitsMax=<0 8 40>, angularLimitsMin=<0 -8 -15>, angularTarget=<0 0 0>}
      localJointOffset=<-6.5 0 0>
     }]
    planarLimits [{drivingBone {boneName=clavicle_l}, planeTransform {location=<0 -2 0>, rotation=<rot -90 90 0>}}]
    sphericalLimits [
     {limitRadius=59, drivingBone {boneName=spine_05}, sphereLocalOffset=<-48 -6.5 0>}
     {limitRadius=6, drivingBone {boneName=spine_05}, sphereLocalOffset=<13 0 0>}
    ]
   }
   n35 = Anim Dynamics - Bone: dyn_hair_l_03(ComponentPose=@n34, Alpha=@n1, ExternalForce="0,0,0") {
    angularDampingOverride=0.99
    angularSpringConstant=1000
    bAngularSpring=true
    bGravityOverrideInSimSpace=true
    bOverrideAngularDamping=true
    bUsePlanarLimit=false
    bUseSphericalLimits=true
    gravityScale=0.35
    alphaBoolBlend {blendOption=Linear}
    boundBone {boneName=dyn_hair_l_03}
    componentAppliedLinearAccClamp=<70 70 70>
    componentLinearVelScale=<300 300 300>
    gravityOverride {z=-980}
    physicsBodyDefinitions [{
      collisionType=CustomSphere
      sphereCollisionRadius=3
      boundBone {boneName=dyn_hair_l_03}
      boxExtents=<5 30 30>
      constraintSetup {angularLimitsMax=<0 10 45>, angularLimitsMin=<0 -10 -25>, angularTarget=<0 0 0>}
      localJointOffset=<-6.5 0 0>
     }]
    sphericalLimits [
     {limitRadius=60, drivingBone {boneName=spine_05}, sphereLocalOffset=<-48 -6.5 0>}
     {limitRadius=8, drivingBone {boneName=spine_05}, sphereLocalOffset=<13 0 0>}
    ]
   }
   n36 = Anim Dynamics - Bone: dyn_hair_r_01(ComponentPose=@n35, Alpha=@n1, ExternalForce="0,0,0") {
    angularDampingOverride=0.99
    angularSpringConstant=1000
    bAngularSpring=true
    bGravityOverrideInSimSpace=true
    bOverrideAngularDamping=true
    bUsePlanarLimit=false
    gravityScale=0.35
    alphaBoolBlend {blendOption=Linear}
    boundBone {boneName=dyn_hair_r_01}
    componentAppliedLinearAccClamp=<70 70 70>
    componentLinearVelScale=<300 300 300>
    gravityOverride {z=-980}
    physicsBodyDefinitions [{
      collisionType=CustomSphere
      sphereCollisionRadius=3
      boundBone {boneName=dyn_hair_r_01}
      boxExtents=<5 30 30>
      constraintSetup {angularLimitsMax=<0 20 17>, angularLimitsMin=<0 -2.5 -8>, angularTarget=<0 0 0>}
      localJointOffset=<-6.5 0 0>
     }]
    sphericalLimits [
     {limitRadius=8, drivingBone {boneName=neck_02}, sphereLocalOffset=<0 0 0>}
     {limitRadius=8, drivingBone {boneName=neck_01}, sphereLocalOffset=<5 10 0>}
    ]
   }
   n37 = Anim Dynamics - Bone: dyn_hair_r_02(ComponentPose=@n36, Alpha=@n1, ExternalForce="0,0,0") {
    angularDampingOverride=0.99
    angularSpringConstant=200
    bAngularSpring=true
    bGravityOverrideInSimSpace=true
    bOverrideAngularDamping=true
    bUseSphericalLimits=true
    gravityScale=0.35
    alphaBoolBlend {blendOption=Linear}
    boundBone {boneName=dyn_hair_r_02}
    componentAppliedLinearAccClamp=<70 70 70>
    componentLinearVelScale=<300 300 300>
    gravityOverride {z=-980}
    physicsBodyDefinitions [{
      collisionType=CustomSphere
      sphereCollisionRadius=2.25
      boundBone {boneName=dyn_hair_r_02}
      boxExtents=<10 30 30>
      constraintSetup {angularLimitsMax=<0 20 20>, angularLimitsMin=<0 -5 -20>, angularTarget=<0 0 0>}
      localJointOffset=<-6.5 0 0>
     }]
    planarLimits [{drivingBone {boneName=clavicle_l}, planeTransform {location=<0 -2 0>, rotation=<rot -90 90 0>}}]
    sphericalLimits [
     {limitRadius=60, drivingBone {boneName=spine_05}, sphereLocalOffset=<-48 -6.5 0>}
     {limitRadius=6.5, drivingBone {boneName=spine_05}, sphereLocalOffset=<13 0 0>}
    ]
   }
   n38 = "Transform (Modify) Bone - Bone: dyn_hair_r_02"(ComponentPose=@n37, Alpha=1, Translation="0,0,1", Rotation="-5,-5,-8", Scale="1,1,1") {
    rotationMode=BMM_Additive
    rotationSpace=BCS_BoneSpace
    translationMode=BMM_Additive
    alphaBoolBlend {blendOption=Linear}
    boneToModify {boneName=dyn_hair_r_02}
    rotation=<rot -5 -5 -8>
    translation {z=1}
   }
   Output Pose(Result=Component To Local(ComponentPose=@n38)) {layerGroup=None}
   n39 = Flap Node - Bone: dyn_hair_bk_01(Alpha=1) {
    bUseClamp=true
    targetClampMax=100
    targetComponent=RotationZ
    alphaBoolBlend {blendOption=Linear}
    sourceDrives [{
      bUseRange=false
      drivingCurve=null
      multiplier=3
      rangeMax=1
      rangeMin=0
      remappedMax=1
      remappedMin=0
      sourceComponent=RotationZ
      useQuaternion=true
      sourceBone {boneName=head}
     }]
    targetBone {boneName=dyn_hair_bk_01}
   }
   n40 = Flap Node - Bone: dyn_hair_bk_l_01(ComponentPose=@n39, Alpha=1) {
    bUseClamp=true
    targetClampMax=100
    targetComponent=RotationZ
    alphaBoolBlend {blendOption=Linear}
    sourceDrives [{
      bUseRange=false
      drivingCurve=null
      multiplier=3
      rangeMax=1
      rangeMin=0
      remappedMax=1
      remappedMin=0
      sourceComponent=RotationZ
      useQuaternion=true
      sourceBone {boneName=head}
     }]
    targetBone {boneName=dyn_hair_bk_l_01}
   }
   Flap Node - Bone: dyn_hair_bk_r_03!(ComponentPose:Component Space Pose Link Structure=@n40, Alpha:float=1) -> (Pose:Component Space Pose Link Structure) {
    is=title FortAnimGraphNode_Flap
    bUseClamp=true
    targetClampMax=100
    targetComponent=RotationZ
    alphaBoolBlend {blendOption=Linear}
    sourceDrives [{
      bUseRange=false
      drivingCurve=null
      multiplier=3
      rangeMax=1
      rangeMin=0
      remappedMax=1
      remappedMin=0
      sourceComponent=RotationZ
      useQuaternion=true
      sourceBone {boneName=head}
     }]
    targetBone {boneName=dyn_hair_bk_r_03}
   }
   n41 = Flap Node - Bone: dyn_hair_bk_02(Alpha=1) {
    bUseClamp=true
    targetClampMax=100
    targetClampMin=-5
    targetComponent=TranslationX
    alphaBoolBlend {blendOption=Linear}
    sourceDrives [{
      bUseRange=false
      drivingCurve=null
      multiplier=-6
      rangeMax=1
      rangeMin=0
      remappedMax=1
      remappedMin=0
      sourceComponent=RotationZ
      useQuaternion=true
      sourceBone {boneName=head}
     }]
    targetBone {boneName=dyn_hair_bk_02}
   }
   n42 = Flap Node - Bone: dyn_hair_bk_l_02(ComponentPose=@n41, Alpha=1) {
    bUseClamp=true
    targetClampMax=100
    targetClampMin=-5
    targetComponent=TranslationX
    alphaBoolBlend {blendOption=Linear}
    sourceDrives [{
      bUseRange=false
      drivingCurve=null
      multiplier=-6
      rangeMax=1
      rangeMin=0
      remappedMax=1
      remappedMin=0
      sourceComponent=RotationZ
      useQuaternion=true
      sourceBone {boneName=head}
     }]
    targetBone {boneName=dyn_hair_bk_l_02}
   }
   Flap Node - Bone: dyn_hair_bk_r_02(ComponentPose=@n42, Alpha=1) {
    bUseClamp=true
    targetClampMax=100
    targetClampMin=-5
    targetComponent=TranslationX
    alphaBoolBlend {blendOption=Linear}
    sourceDrives [{
      bUseRange=false
      drivingCurve=null
      multiplier=-6
      rangeMax=1
      rangeMin=0
      remappedMax=1
      remappedMin=0
      sourceComponent=RotationZ
      useQuaternion=true
      sourceBone {boneName=head}
     }]
    targetBone {boneName=dyn_hair_bk_r_02}
   }
  ]}
 EventGraph {nodes [
   Event BlueprintUpdateAnimation()
   n1 = MapRangeClamped(Value=Get PawnSpeed2D(), InRangeA=0, InRangeB=100)
   > n2 = Branch(Condition="float <= float"!(A:double=@n1, B:double=0) -> (ReturnValue:bool) {is=title K2Node_CallFunction})
   > Set speed_mult(speed_mult:double=0.1) -> (Output_Get:double)
   > Set hair_alpha(hair_alpha:double=0.65) -> (Output_Get:double)
   Set speed_mult(execute=@n2.else, speed_mult=float * float(A=@n1, B=0.65))
   n3 = "Sin (Degrees)"!(A:double=float * float(A=GetGameTimeInSeconds(), B=250)) -> (ReturnValue:double)
   float * float(A=@n3, B=20)
   Set copy_alpha(copy_alpha:double="Clamp (Float)"(Value=@n3, Min=0.1, Max=1)) -> (Output_Get:double)
  ]}
]
variables [
 copy_alpha {type=number}
 hair_alpha {type=number}
 speed_mult {type=number}
]
